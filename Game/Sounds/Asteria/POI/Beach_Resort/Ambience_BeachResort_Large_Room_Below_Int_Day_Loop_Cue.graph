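# Looping indoor room-tone ambience cue for the large below-level interior of the Beach Resort POI by day, using a large stone room recording.
# A quality-level node selects the quad wave for its first two slots and the stereo wave for its last two, each through a fixed pitch-up modulator.
# It plays on the indoor ambient sound class without attenuation, spatialization or reverb send, and concurrency prevents new instances.
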SoundCue /Game/Sounds/Asteria/POI/Beach_Resort/Ambience_BeachResort_Large_Room_Below_Int_Day_Loop_Cue
root=QualityLevel_1
nodes [
 {id=Modulator_0, children [WavePlayer_8], properties {pitchMax=1.2, pitchMin=1.2}}
 {id=Modulator_3, children [WavePlayer_9], properties {pitchMax=1.2, pitchMin=1.2}}
 {id=QualityLevel_1, children [Modulator_0,Modulator_0,Modulator_3,Modulator_3]}
 {
  id=WavePlayer_8
  properties {bLooping=true, soundWaveAssetPtr=/Game/Sounds/Artemis/RoomTones/Stone/Artemis_Interior_Large_Stone_Room_02_Quad}
 }
 {
  id=WavePlayer_9
  properties {bLooping=true, soundWaveAssetPtr=/Game/Sounds/Artemis/RoomTones/Stone/Artemis_Interior_Large_Stone_Room_02_Stereo}
 }
]
properties {
 bOverrideConcurrency=true
 maxDistance=2097152
 pitchMultiplier=0.9
 soundClassObject=&Fort_Ambient_Indoors_Athena
 volumeMultiplier=0.501
 attenuationOverrides {bAttenuate=false, bEnableReverbSend=false, bSpatialize=false}
 concurrencyOverrides {resolutionRule=PreventNew}
}
