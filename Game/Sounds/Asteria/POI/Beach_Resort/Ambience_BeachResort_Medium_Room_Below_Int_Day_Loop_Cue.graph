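# Sound cue for the looping daytime interior ambience of a medium room at the Beach Resort POI, reusing Residential interior recordings.
# A quality-level switch picks the looping quad wave for the first two levels and the looping stereo wave for the last two.
# It plays non-spatialized in the indoor ambient sound class, pitched slightly down and boosted in volume.
# Its concurrency override prevents new instances once the limit is reached.
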SoundCue /Game/Sounds/Asteria/POI/Beach_Resort/Ambience_BeachResort_Medium_Room_Below_Int_Day_Loop_Cue
root=QualityLevel_0
nodes [
 {id=QualityLevel_0, children [WavePlayer_5,WavePlayer_5,WavePlayer_4,WavePlayer_4]}
 {id=WavePlayer_4, properties {bLooping=true, soundWaveAssetPtr=/Game/Sounds/Asteria/POI/Residential/Ambience_Residential_Interior_Stereo_3}}
 {id=WavePlayer_5, properties {bLooping=true, soundWaveAssetPtr=/Game/Sounds/Asteria/POI/Residential/Ambience_Residential_Interior_Quad_3}}
]
properties {
 bOverrideConcurrency=true
 maxDistance=2097152
 pitchMultiplier=0.9
 soundClassObject=&Fort_Ambient_Indoors_Athena
 volumeMultiplier=1.2
 attenuationOverrides {bAttenuate=false, bEnableReverbSend=false, bSpatialize=false}
 concurrencyOverrides {resolutionRule=PreventNew}
}
