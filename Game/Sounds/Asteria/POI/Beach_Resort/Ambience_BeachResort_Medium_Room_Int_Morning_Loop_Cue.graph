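# Sound Cue for the morning indoor ambience of a medium room at the Beach Resort POI, assigned to the indoor ambient sound class.
# A quality switch picks a quad mix for its first two levels and a stereo mix for its last two.
# Each mix layers a looping residential interior bed at full volume with a quieter resort morning exterior bed.
# It plays without attenuation, spatialization or reverb send, and its concurrency override prevents new instances.
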SoundCue /Game/Sounds/Asteria/POI/Beach_Resort/Ambience_BeachResort_Medium_Room_Int_Morning_Loop_Cue
root=QualityLevel_0
nodes [
 {id=Mixer_1, children [WavePlayer_1,WavePlayer_5], properties {inputVolume [0.3,1]}}
 {id=Mixer_3, children [WavePlayer_4,WavePlayer_0], properties {inputVolume [1,0.3]}}
 {id=QualityLevel_0, children [Mixer_1,Mixer_1,Mixer_3,Mixer_3]}
 {id=WavePlayer_0, properties {soundWaveAssetPtr=&/Ambience_BeachResort_Morning_Exterior_Loop_Stereo}}
 {id=WavePlayer_1, properties {soundWaveAssetPtr=&/Ambience_BeachResort_Morning_Exterior_Loop_Quad}}
 {id=WavePlayer_4, properties {bLooping=true, soundWaveAssetPtr=/Game/Sounds/Asteria/POI/Residential/Ambience_Residential_Interior_Stereo_3}}
 {id=WavePlayer_5, properties {bLooping=true, soundWaveAssetPtr=/Game/Sounds/Asteria/POI/Residential/Ambience_Residential_Interior_Quad_3}}
]
properties {
 bOverrideConcurrency=true
 maxDistance=2097152
 soundClassObject=&Fort_Ambient_Indoors_Athena
 volumeMultiplier=1.403
 attenuationOverrides {bAttenuate=false, bEnableReverbSend=false, bSpatialize=false}
 concurrencyOverrides {resolutionRule=PreventNew}
}
