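# Looping daytime basement ambience Sound Cue for the Brawney House at the Beach Resort POI.
# A single looping Wave Player is the root, reusing the Hollow Base basement day loop wave from the Artemis sound folder with a boosted volume.
# It plays non-spatialized and unattenuated under the Athena indoor ambient sound class, with bus and submix sends off.
# A concurrency override prevents new instances once the limit is reached.
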SoundCue /Game/Sounds/Asteria/POI/Beach_Resort/Ambience_BrawneyHouse_Basement_Day_Loop_Cue
root=WavePlayer_1
nodes [{id=WavePlayer_1, properties {bLooping=true, soundWaveAssetPtr=/Game/Sounds/Artemis/POI/Hollow_Base/Ambience_Hollow_Base_Basement_Day_Loop}}]
properties {
 bEnableBusSends=false
 bEnableSubmixSends=false
 bOverrideConcurrency=true
 maxDistance=2097152
 soundClassObject=&Fort_Ambient_Indoors_Athena
 volumeMultiplier=3.007
 attenuationOverrides {bAttenuate=false, bEnableReverbSend=false, bSpatialize=false}
 concurrencyOverrides {resolutionRule=PreventNew}
}
