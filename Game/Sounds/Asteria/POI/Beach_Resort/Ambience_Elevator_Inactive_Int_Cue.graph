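# Looping interior ambience cue for an inactive elevator at the Beach Resort POI, on the indoor ambient sound class.
# A quality level node selects between two looping wave players, using the quad loop for the first two levels and the stereo loop for the last two.
# It plays at reduced volume without attenuation, spatialization or reverb send, and its concurrency override prevents new instances once the limit is reached.
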SoundCue /Game/Sounds/Asteria/POI/Beach_Resort/Ambience_Elevator_Inactive_Int_Cue
root=QualityLevel_0
nodes [
 {id=QualityLevel_0, children [WavePlayer_4,WavePlayer_4,WavePlayer_5,WavePlayer_5]}
 {id=WavePlayer_4, properties {bLooping=true, soundWaveAssetPtr=&/Ambience_Elevator_Inactive_Loop_Quad}}
 {id=WavePlayer_5, properties {bLooping=true, soundWaveAssetPtr=&/Ambience_Elevator_Inactive_Loop_Stereo}}
]
properties {
 bOverrideConcurrency=true
 maxDistance=2097152
 soundClassObject=&Fort_Ambient_Indoors_Athena
 volumeMultiplier=0.301
 attenuationOverrides {bAttenuate=false, bEnableReverbSend=false, bSpatialize=false}
 concurrencyOverrides {resolutionRule=PreventNew}
}
