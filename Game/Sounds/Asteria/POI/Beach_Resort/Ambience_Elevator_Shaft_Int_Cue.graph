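# Sound cue for the interior elevator shaft ambience loop at the Beach Resort POI.
# A quality level node picks between two looping wave players, using the quad loop for the first two levels and the stereo loop for the other two.
# It plays through the indoor ambient sound class at reduced volume, with attenuation, spatialization and reverb send disabled in its overrides.
# Its concurrency override prevents new instances when the limit is reached.
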SoundCue /Game/Sounds/Asteria/POI/Beach_Resort/Ambience_Elevator_Shaft_Int_Cue
root=QualityLevel_0
nodes [
 {id=QualityLevel_0, children [WavePlayer_6,WavePlayer_6,WavePlayer_7,WavePlayer_7]}
 {id=WavePlayer_6, properties {bLooping=true, soundWaveAssetPtr=&/Ambience_Elevator_Shaft_Loop_Quad}}
 {id=WavePlayer_7, properties {bLooping=true, soundWaveAssetPtr=&/Ambience_Elevator_Shaft_Loop_Stereo}}
]
properties {
 bOverrideConcurrency=true
 maxDistance=2097152
 soundClassObject=&Fort_Ambient_Indoors_Athena
 volumeMultiplier=0.501
 attenuationOverrides {bAttenuate=false, bEnableReverbSend=false, bSpatialize=false}
 concurrencyOverrides {resolutionRule=PreventNew}
}
