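# Sound cue for the looping interior kitchen room-tone ambience at the Beach Resort POI.
# A quality-level node selects a looping quad wave for the first two quality slots and a looping stereo wave for the last two.
# It plays quietly in the indoor ambient sound class as a non-spatialized, non-attenuated bed with no reverb send.
# A concurrency override prevents new instances once the limit is reached.
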SoundCue /Game/Sounds/Asteria/POI/Beach_Resort/Ambience_Kitchen_Room_Tone_Int_Cue
root=QualityLevel_0
nodes [
 {id=QualityLevel_0, children [WavePlayer_0,WavePlayer_0,WavePlayer_1,WavePlayer_1]}
 {id=WavePlayer_0, properties {bLooping=true, soundWaveAssetPtr=&/Ambience_Kitchen_Room_Tone_Quad}}
 {id=WavePlayer_1, properties {bLooping=true, soundWaveAssetPtr=&/Ambience_Kitchen_Room_Tone_Stereo}}
]
properties {
 bOverrideConcurrency=true
 maxDistance=2097152
 soundClassObject=&Fort_Ambient_Indoors_Athena
 volumeMultiplier=0.2
 attenuationOverrides {bAttenuate=false, bEnableReverbSend=false, bSpatialize=false}
 concurrencyOverrides {resolutionRule=PreventNew}
}
