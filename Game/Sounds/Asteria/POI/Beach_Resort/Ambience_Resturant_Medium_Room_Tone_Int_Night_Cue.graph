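# Looping, non-spatialized night interior ambience cue for a medium restaurant room at the Beach Resort POI.
# A quality-level node selects quad waves for the first two levels and stereo waves for the last two.
# Each mixer layers full-volume room tone with the night exterior loop at reduced volume.
# The cue plays slightly pitched down on the Athena ambient sound class, with concurrency set to prevent new instances past the limit.
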SoundCue /Game/Sounds/Asteria/POI/Beach_Resort/Ambience_Resturant_Medium_Room_Tone_Int_Night_Cue
root=QualityLevel_0
nodes [
 {id=Mixer_1, children [WavePlayer_2,WavePlayer_0], properties {inputVolume [0.25,1]}}
 {id=Mixer_3, children [WavePlayer_1,WavePlayer_3], properties {inputVolume [1,0.25]}}
 {id=QualityLevel_0, children [Mixer_1,Mixer_1,Mixer_3,Mixer_3]}
 {
  id=WavePlayer_0
  properties {bLooping=true, soundWaveAssetPtr=&/Ambience_Resturant_Medium_Room_Tone_Quad}
 }
 {
  id=WavePlayer_1
  properties {bLooping=true, soundWaveAssetPtr=&/Ambience_Resturant_Medium_Room_Tone_Stereo}
 }
 {
  id=WavePlayer_2
  properties {bLooping=true, soundWaveAssetPtr=&/Ambience_BeachResort_Night_Exterior_Loop_Quad}
 }
 {
  id=WavePlayer_3
  properties {bLooping=true, soundWaveAssetPtr=&/Ambience_BeachResort_Night_Exterior_Loop_Stereo}
 }
]
properties {
 bOverrideConcurrency=true
 maxDistance=2097152
 pitchMultiplier=0.9
 soundClassObject=&Fort_Ambient_Athena
 volumeMultiplier=1.103
 attenuationOverrides {bAttenuate=false, bEnableReverbSend=false, bSpatialize=false}
 concurrencyOverrides {resolutionRule=PreventNew}
}
